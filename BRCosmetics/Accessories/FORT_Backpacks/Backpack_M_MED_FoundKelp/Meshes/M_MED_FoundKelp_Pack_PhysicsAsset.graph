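# Physics asset for the M_MED_FoundKelp backpack, setting up its cape bones for physically simulated motion.
# Three capsule chains (center, left, right) hang from MSTR bones, with tighter swing near the roots and locked twist on upper links.
# Angularly free cross constraints tie the center chain to its sides, unevenly: five links on the right, only two lower ones on the left.
# Torso, pelvis and right-side limb capsules supply body shapes.
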PhysicsAsset /BRCosmetics/Accessories/FORT_Backpacks/Backpack_M_MED_FoundKelp/Meshes/M_MED_FoundKelp_Pack_PhysicsAsset
bodies {
 calf_r [{shapeType=Capsule, length=11.7963, radius=39.0787, center=<26.1258 -24.118 0>, rotation=<rot 90 180 1.4887>}]
 clavicle_r [{shapeType=Capsule, length=12.9688, radius=9.6562, center=<-12.3638 1.6165 0>, rotation=<rot 90 0 173.4389>}]
 dyn_cape_01 [{shapeType=Capsule, length=11.7556, radius=4.9372, shapeName=dyn_cape_01_capsule, center=<7.1136 0 3.9901>, rotation=<rot 86.6514 0 0>}]
 dyn_cape_02 [{shapeType=Capsule, length=11.3116, radius=5.1687, shapeName=dyn_cape_01_capsule, center=<7.1588 0 3.3167>, rotation=<rot 85.7986 0 0>}]
 dyn_cape_03 [{shapeType=Capsule, length=11.3116, radius=5.1687, shapeName=dyn_cape_01_capsule, center=<7.15 0 2.3>, rotation=<rot 90 0 0>}]
 dyn_cape_04 [{shapeType=Capsule, length=11.3116, radius=5.1687, shapeName=dyn_cape_01_capsule, center=<7.15 0 1.2466>, rotation=<rot 90 0.1804 0>}]
 dyn_cape_05 [{shapeType=Capsule, length=11.3116, radius=5.1687, shapeName=dyn_cape_01_capsule, center=<7.15 0 1.4396>, rotation=<rot 90 0 0>}]
 dyn_cape_06 [{shapeType=Capsule, length=11.3116, radius=5.1687, shapeName=dyn_cape_01_capsule, center=<7.15 0 1.4396>, rotation=<rot 90 0 0>}]
 dyn_cape_07 [{shapeType=Capsule, length=3.0762, radius=5.1687, shapeName=dyn_cape_01_capsule, center=<5.8445 0 1.4396>, rotation=<rot 90 0 0>}]
 dyn_cape_MSTR [{shapeType=Capsule, length=0.505, radius=0.505, shapeName=dyn_cape_r_MSTR_capsule, rotation=<rot 0 -90 0>}]
 dyn_cape_l_01 [{shapeType=Capsule, length=11.7556, radius=4.9372, shapeName=dyn_cape_01_capsule, center=<7.1535 0 5.5396>, rotation=<rot 86.6514 0 0>}]
 dyn_cape_l_02 [{shapeType=Capsule, length=11.3116, radius=5.1687, shapeName=dyn_cape_01_capsule, center=<7.1576 0 3.3077>, rotation=<rot 82.2633 0 0>}]
 dyn_cape_l_03 [{shapeType=Capsule, length=11.3116, radius=5.1687, shapeName=dyn_cape_01_capsule, center=<7.15 0 1.1042>, rotation=<rot 90 0 0>}]
 dyn_cape_l_04 [{shapeType=Capsule, length=11.3116, radius=5.1687, shapeName=dyn_cape_01_capsule, center=<7.15 0 0.0988>, rotation=<rot 90 0 0>}]
 dyn_cape_l_05 [{shapeType=Capsule, length=11.3116, radius=5.1687, shapeName=dyn_cape_01_capsule, center=<7.15 0 1.4396>, rotation=<rot 90 0 0>}]
 dyn_cape_l_06 [{shapeType=Capsule, length=11.3116, radius=5.1687, shapeName=dyn_cape_01_capsule, center=<7.15 0 1.4396>, rotation=<rot 90 0 0>}]
 dyn_cape_l_07 [{shapeType=Capsule, length=3.0762, radius=5.1687, shapeName=dyn_cape_01_capsule, center=<5.8445 0 1.4396>, rotation=<rot 90 0 0>}]
 dyn_cape_l_MSTR [{shapeType=Capsule, length=0.505, radius=0.505, shapeName=dyn_cape_r_MSTR_capsule, rotation=<rot 0 -90 0>}]
 dyn_cape_r_01 [{shapeType=Capsule, length=11.7556, radius=4.9372, shapeName=dyn_cape_01_capsule, center=<7.0218 0 2.5302>, rotation=<rot 86.4778 0 0>}]
 dyn_cape_r_02 [{shapeType=Capsule, length=11.3116, radius=5.1687, shapeName=dyn_cape_01_capsule, center=<7.0413 0 1.7169>, rotation=<rot 85.7986 0 0>}]
 dyn_cape_r_03 [{shapeType=Capsule, length=11.3116, radius=5.1687, shapeName=dyn_cape_01_capsule, center=<7.15 0 1.5082>, rotation=<rot 90 0 0>}]
 dyn_cape_r_04 [{shapeType=Capsule, length=11.3116, radius=5.1687, shapeName=dyn_cape_01_capsule, center=<7.15 0 1.2466>, rotation=<rot 90 0 0>}]
 dyn_cape_r_05 [{shapeType=Capsule, length=11.3116, radius=5.1687, shapeName=dyn_cape_01_capsule, center=<7.15 0 1.4396>, rotation=<rot 90 0 0>}]
 dyn_cape_r_06 [{shapeType=Capsule, length=11.3116, radius=5.1687, shapeName=dyn_cape_01_capsule, center=<7.15 0 1.4396>, rotation=<rot 90 0 0>}]
 dyn_cape_r_07 [{shapeType=Capsule, length=3.0762, radius=5.1687, shapeName=dyn_cape_01_capsule, center=<5.8445 0 1.4396>, rotation=<rot 90 0 0>}]
 dyn_cape_r_MSTR [{shapeType=Capsule, length=0.505, radius=0.505, shapeName=dyn_cape_r_MSTR_capsule, rotation=<rot 0 -90 0>}]
 pelvis [{shapeType=Capsule, length=6.0938, radius=14.75, center=<-0.9978 -1.6863 0>}]
 spine_03 [{shapeType=Capsule, length=39.6562, radius=18, center=<2.5875 2.4605 0>}]
 spine_04 [{shapeType=Capsule, length=31.4062, radius=17.4375, center=<2.8832 2.8193 0>}]
 spine_05 [{shapeType=Capsule, length=32.875, radius=15.375, center=<-0.0029 0.2322 0>}]
 thigh_r [{shapeType=Capsule, length=28.0938, radius=9.2188, center=<19.7778 0.1788 0>, rotation=<rot 90 -10.6197 166.7532>}]
 upperarm_r [{shapeType=Capsule, length=32.9087, radius=10.2407, center=<-20.6057 -0.0908 0>, rotation=<rot 90 -173.4389 0>}]
}
constraints [
 {bone1Name=dyn_cape_r_01, bone2Name=dyn_cape_r_MSTR, swing1LimitDegrees=10, swing2LimitDegrees=30, twistLimitDegrees=5, twistMotion=Locked}
 {bone1Name=dyn_cape_r_02, bone2Name=dyn_cape_r_01, swing1LimitDegrees=20, swing2LimitDegrees=35, twistLimitDegrees=5, twistMotion=Locked}
 {bone1Name=dyn_cape_r_03, bone2Name=dyn_cape_r_02, swing1LimitDegrees=35, swing2LimitDegrees=30, twistLimitDegrees=5, twistMotion=Locked}
 {bone1Name=dyn_cape_r_04, bone2Name=dyn_cape_r_03, swing1LimitDegrees=35, swing2LimitDegrees=30, twistLimitDegrees=5}
 {bone1Name=dyn_cape_r_05, bone2Name=dyn_cape_r_04, swing1LimitDegrees=35, swing2LimitDegrees=30, twistLimitDegrees=5}
 {bone1Name=dyn_cape_r_06, bone2Name=dyn_cape_r_05, swing1LimitDegrees=35, swing2LimitDegrees=30, twistLimitDegrees=5}
 {bone1Name=dyn_cape_r_07, bone2Name=dyn_cape_r_06, swing1LimitDegrees=35, swing2LimitDegrees=30, twistLimitDegrees=5}
 {bone1Name=dyn_cape_l_01, bone2Name=dyn_cape_l_MSTR, swing1LimitDegrees=10, swing2LimitDegrees=30, twistLimitDegrees=5, twistMotion=Locked}
 {bone1Name=dyn_cape_l_02, bone2Name=dyn_cape_l_01, swing1LimitDegrees=20, swing2LimitDegrees=35, twistLimitDegrees=5, twistMotion=Locked}
 {bone1Name=dyn_cape_l_03, bone2Name=dyn_cape_l_02, swing1LimitDegrees=35, swing2LimitDegrees=30, twistLimitDegrees=5, twistMotion=Locked}
 {bone1Name=dyn_cape_l_04, bone2Name=dyn_cape_l_03, swing1LimitDegrees=35, swing2LimitDegrees=30, twistLimitDegrees=5}
 {bone1Name=dyn_cape_l_05, bone2Name=dyn_cape_l_04, swing1LimitDegrees=35, swing2LimitDegrees=30, twistLimitDegrees=5}
 {bone1Name=dyn_cape_l_06, bone2Name=dyn_cape_l_05, swing1LimitDegrees=35, swing2LimitDegrees=30, twistLimitDegrees=5}
 {bone1Name=dyn_cape_l_07, bone2Name=dyn_cape_l_06, swing1LimitDegrees=35, swing2LimitDegrees=30, twistLimitDegrees=5}
 {bone1Name=dyn_cape_01, bone2Name=dyn_cape_MSTR, swing1LimitDegrees=10, swing2LimitDegrees=30, twistLimitDegrees=5, twistMotion=Locked}
 {bone1Name=dyn_cape_02, bone2Name=dyn_cape_01, swing1LimitDegrees=20, swing2LimitDegrees=35, twistLimitDegrees=5, twistMotion=Locked}
 {bone1Name=dyn_cape_03, bone2Name=dyn_cape_02, swing1LimitDegrees=35, swing2LimitDegrees=30, twistLimitDegrees=5, twistMotion=Locked}
 {bone1Name=dyn_cape_04, bone2Name=dyn_cape_03, swing1LimitDegrees=35, swing2LimitDegrees=30, twistLimitDegrees=5}
 {bone1Name=dyn_cape_05, bone2Name=dyn_cape_04, swing1LimitDegrees=35, swing2LimitDegrees=30, twistLimitDegrees=5}
 {bone1Name=dyn_cape_06, bone2Name=dyn_cape_05, swing1LimitDegrees=35, swing2LimitDegrees=30, twistLimitDegrees=5}
 {bone1Name=dyn_cape_07, bone2Name=dyn_cape_06, swing1LimitDegrees=35, swing2LimitDegrees=30, twistLimitDegrees=5}
 {
  bone1Name=dyn_cape_r_05
  bone2Name=dyn_cape_05
  swing1LimitDegrees=45
  swing1Motion=Free
  swing2LimitDegrees=45
  swing2Motion=Free
  twistLimitDegrees=45
  twistMotion=Free
 }
 {
  bone1Name=dyn_cape_l_05
  bone2Name=dyn_cape_05
  swing1LimitDegrees=45
  swing1Motion=Free
  swing2LimitDegrees=45
  swing2Motion=Free
  twistLimitDegrees=45
  twistMotion=Free
 }
 {
  bone1Name=dyn_cape_l_07
  bone2Name=dyn_cape_07
  swing1LimitDegrees=45
  swing1Motion=Free
  swing2LimitDegrees=45
  swing2Motion=Free
  twistLimitDegrees=45
  twistMotion=Free
 }
 {
  bone1Name=dyn_cape_r_07
  bone2Name=dyn_cape_07
  swing1LimitDegrees=45
  swing1Motion=Free
  swing2LimitDegrees=45
  swing2Motion=Free
  twistLimitDegrees=45
  twistMotion=Free
 }
 {
  bone1Name=dyn_cape_03
  bone2Name=dyn_cape_r_03
  swing1LimitDegrees=45
  swing1Motion=Free
  swing2LimitDegrees=45
  swing2Motion=Free
  twistLimitDegrees=45
  twistMotion=Free
 }
 {
  bone1Name=dyn_cape_04
  bone2Name=dyn_cape_r_04
  swing1LimitDegrees=45
  swing1Motion=Free
  swing2LimitDegrees=45
  swing2Motion=Free
  twistLimitDegrees=45
  twistMotion=Free
 }
 {
  bone1Name=dyn_cape_02
  bone2Name=dyn_cape_r_02
  swing1LimitDegrees=45
  swing1Motion=Free
  swing2LimitDegrees=45
  swing2Motion=Free
  twistLimitDegrees=45
  twistMotion=Free
 }
]
